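# Translucent unlit Niagara sprite material for the LatteStir emote that distorts and tints the scene color behind its particles.
# It sums offset SceneColor samples scaled by BlurIntensity around the particle center, then multiplies by an RGB-split sample driven by dynamic Offset and particle color.
# Scalar parameters blend toward a plain glow or tinted scene color.
# A soft circular mask, particle alpha and depth fade set opacity; Forward and Mobile paths output constants instead.
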Material /BRCosmetics/Effects/Fort_Effects/Emotes/LatteStir/Materials/MI_Emote_LatteStir_SceneColor
graph {
 properties {
  bAutomaticallySetUsageInEditor=false
  bUsedWithNiagaraSprites=true
  blendMode=BLEND_Translucent
  shadingModel=MSM_Unlit
  usedShadingModels=Unlit
 }
 nodes [
  n1 = ParticleColor()
  n2 = SphereMask(A=Add(A=TextureCoordinate()) {constB=0}, B=Constant2Vector() {g=0.5, r=0.5}) {attenuationRadius=0.4, hardnessPercent=30}
  n3 = DepthFade(Opacity=Multiply(A=@n1.A, B=Power(Base=@n2)), FadeDistance=ScalarParameter() {defaultValue=100, parameterName=FadeDistance})
  n4 = Constant()
  Reroute(None=ShadingPathSwitch(Default=@n3, Deferred=@n3, Forward=@n4, Mobile=@n4))
  n5 = Constant()
  n6 = MaterialFunctionCall!(Center Point World Position=ParticlePositionWS()) {materialFunction=/Engine/Functions/Engine_MaterialFunctions02/UVs/WorldSpaceAlignedScreenCoordinates}
  n7 = Multiply(A=ScalarParameter() {defaultValue=10, parameterName=BlurIntensity}, B=Constant() {r=0.03})
  n8 = Add(
   A=SceneColor!(OffsetFraction=Multiply(A=@n6.Offset and Scaled UVS) {constB=0}) {inputMode=OffsetFraction}.RGB
   B=SceneColor(OffsetFraction=Multiply(A=@n6.Offset and Scaled UVS, B=@n7)) {inputMode=OffsetFraction}.RGB
  )
  n9 = ScalarParameter() {defaultValue=10, parameterName=BlurIntensity}
  n10 = Constant() {r=0.06}
  n11 = SceneColor(OffsetFraction=Multiply(A=@n6.Offset and Scaled UVS, B=Multiply(A=@n9, B=@n10))) {inputMode=OffsetFraction}
  n12 = Multiply(B=Multiply(A=ScalarParameter() {defaultValue=10, parameterName=BlurIntensity}, B=Constant() {r=0.09}))
  n13 = Divide(A=Add(A=Add(A=@n8, B=@n11.RGB), B=SceneColor(OffsetFraction=@n12) {inputMode=OffsetFraction}.RGB)) {constB=1}
  n14 = Multiply(A=Constant2Vector() {g=0.02}, B=Multiply(A=DynamicParameter() {paramNames [Offset,Param2,Param3,Param4]}.Offset) {constB=0.2}) {constB=0.2}
  n15 = SceneColor(OffsetFraction=@n14) {inputMode=OffsetFraction, constInput=<0.02 0.02>}
  n16 = SceneColor() {inputMode=OffsetFraction}
  n17 = SceneColor(OffsetFraction=Multiply(A=@n14) {constB=-1}) {inputMode=OffsetFraction, constInput=<-0.1 -0.1>}
  n18 = Multiply(
   A=MaterialFunctionCall(X=ComponentMask(None=@n15.RGB) {r=true}, Y=ComponentMask(None=@n16.RGB) {g=true}, Z=ComponentMask(None=@n17.RGB) {b=true}) {materialFunction=&MakeFloat3}.Result
   B=@n1.RGB
  )
  n19 = LinearInterpolate(
   A=LinearInterpolate(A=Multiply(A=@n13, B=@n18), B=ParticleColor().RGB, Alpha=ScalarParameter() {parameterName=UseSimpleGlow})
   B=Multiply(A=ParticleColor().RGB, B=SceneColor() {inputMode=OffsetFraction}.RGB)
   Alpha=ScalarParameter() {parameterName=UseSimpleScenecolor}
  )
  ShadingPathSwitch(Default=@n19, Deferred=@n19, Forward=@n5, Mobile=@n5)
 ]
}
