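# Translucent unlit Niagara mesh-particle material for the long flame trail in the LatteStir emote.
# Particle color feeds emissive, while a panned base trail texture, noise dissolve and edge masks shape depth-faded opacity, with an optional HeightLerp final dissolve.
# SplineThicken sets ribbon width, added to one of two blendable sine-wave vertex offsets.
# Dynamic parameters drive width, U offsets, wobble strength, dissolve and PannerSpeed, which replaces time in panners and waves.
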Material /BRCosmetics/Effects/Fort_Effects/Emotes/LatteStir/Materials/M_Emote_LatteStir_LongFlame
graph {
 properties {
  bAutomaticallySetUsageInEditor=false
  bUsedWithNiagaraMeshParticles=true
  blendMode=BLEND_Translucent
  shadingModel=MSM_Unlit
  twoSided=true
  usedShadingModels=Unlit
 }
 nodes [
  n1 = Add(
   A=TextureCoordinate()
   B=AppendVector(A=DynamicParameter() {defaultValue=<rgba 3.5 1 1 1>, paramNames [SplineWidth,U_offset,WPO,Dissolve]}.U_offset, B=Constant())
  )
  n2 = ComponentMask(None=VectorParameter() {parameterName=FinalDissolveUVScale, defaultValue=<rgba 1 0.5 0 0>}.RGB) {g=true, r=true}
  n3 = DynamicParameter() {parameterIndex=1, defaultValue=<rgba 3.5 0 1 1>, paramNames [PannerSpeed, Tex_Base_UOffset, None, None]}
  n4 = VectorParameter() {parameterName=SpeedXY, defaultValue=<rgba -0.5 0 0 0>}
  n5 = TextureSampleParameter2D(UVs=Panner(Coordinate=Multiply(A=@n1, B=@n2), Time=@n3.PannerSpeed, Speed=ComponentMask(None=@n4.RGB) {g=true, r=true})) {parameterName=FinalDissolve, samplerType=SAMPLERTYPE_LinearColor, texture=/Game/Global/Textures/Default/Noise/T_Noise01_Inverse}
  n6 = MaterialFunctionCall(
   Transition Phase=LinearInterpolate(
    A=ScalarParameter() {defaultValue=0.3, parameterName=DissMin}
    B=ScalarParameter() {defaultValue=0.8, parameterName=DissMax}
    Alpha=DynamicParameter() {defaultValue=<rgba 3.5 1 1 1>, paramNames [SplineWidth,U_offset,WPO,Dissolve]}.Dissolve
   ) {constA=0.3, constB=0.8}
   Height Texture=@n5.R
   Contrast=ScalarParameter() {defaultValue=3, parameterName="Contrast (S)"}
  ) {materialFunction=&HeightLerp}
  n7 = ScalarParameter() {parameterName=UseDissolve}
  n8 = ParticleColor()
  n9 = TextureCoordinate() {uTiling=0.7}
  n10 = DynamicParameter() {parameterIndex=1, defaultValue=<rgba 3.5 0 1 1>, paramNames [PannerSpeed, Tex_Base_UOffset, None, None]}
  n11 = Add(A=@n9, B=AppendVector(A=@n10.Tex_Base_UOffset, B=Constant()))
  n12 = ScalarParameter() {parameterName=Use_Tex_Base_UOffset}
  n13 = Power(Base=ComponentMask(None=LinearInterpolate(A=@n9, B=@n11, Alpha=@n12)) {r=true}) {constExponent=0.5}
  n14 = ComponentMask(None=@n9) {g=true}
  n15 = TextureSampleParameter2D(UVs=Panner(Coordinate=AppendVector(A=@n13, B=@n14), Time=@n10.PannerSpeed) {speedX=-0.8}) {parameterName=Tex_Base, texture=&/T_Emote_LatteStir_DarkTrail}
  n16 = ScalarParameter() {defaultValue=-3, parameterName=DissLerpMin}
  n17 = ScalarParameter() {defaultValue=3, parameterName=DissLerpMax}
  n18 = Multiply(
   A=TextureCoordinate()
   B=ComponentMask(None=VectorParameter() {parameterName=DissUVScale, defaultValue=<rgba 2 0.8 0 0>}.RGB) {g=true, r=true}
  )
  n19 = AppendVector(A=DynamicParameter() {defaultValue=<rgba 3.5 1 1 1>, paramNames [SplineWidth,U_offset,WPO,Dissolve]}.U_offset, B=Constant())
  n20 = ComponentMask(None=VectorParameter() {parameterName=DisSpeed, defaultValue=<rgba -2 0 0 0>}.RGB) {g=true, r=true}
  n21 = ScalarParameter() {defaultValue=1, parameterName=Speed}
  n22 = TextureSampleParameter2D(UVs=Panner(Coordinate=Add(A=@n18, B=@n19), Time=@n10.PannerSpeed, Speed=Multiply(A=@n20, B=@n21))) {parameterName=DissolveTexture, texture=/Game/Effects/Niagara/MasterMaterials/SplineThickenTrail/Texture/T_Pickaxe_SoapPocket_NoiseGlow}
  n23 = Constant()
  n24 = Constant() {r=0.6}
  n25 = Power(Base=ComponentMask(None=TextureCoordinate()) {r=true}) {constExponent=1.3}
  n26 = Power(Base=@n22.R, Exp=Saturate(None=LinearInterpolate(A=@n23, B=@n24, Alpha=@n25)))
  n27 = Saturate(None=LinearInterpolate(Alpha=LinearInterpolate(A=@n16, B=@n17, Alpha=@n26)) {constA=-1})
  n28 = OneMinus(None=ComponentMask(None=TextureCoordinate()) {r=true})
  n29 = Saturate(None=Multiply(A=Power(Base=@n28) {constExponent=1}) {constB=7})
  n30 = SphereMask(A=TextureCoordinate() {uTiling=0.7, vTiling=0.7}, B=Constant2Vector() {g=0.35, r=0.5}) {attenuationRadius=0.5, hardnessPercent=50}
  n31 = Multiply(
   A=Multiply(A=StaticComponentMaskParameter(None=@n15.RGB) {defaultR=true, parameterName=Mask}, B=Multiply(A=@n27, B=@n29))
   B=Saturate(None=Multiply(A=@n30) {constB=10})
  )
  DepthFade(
   Opacity=Multiply(A=Multiply(A=@n8.A, B=@n31), B=LinearInterpolate(B=@n6.Results, Alpha=@n7) {constA=1})
   FadeDistance=ScalarParameter() {defaultValue=10, parameterName=FadeDistance}
  )
  MaterialFunctionCall(Emissive=@n8.RGB) {materialFunction=&MF_FX_EmissiveAdjustments}
  Power(Base=ComponentMask(None=TextureCoordinate()) {r=true}) {constExponent=1.5}
  n32 = DynamicParameter() {defaultValue=<rgba 3.5 1 1 1>, paramNames [SplineWidth,U_offset,WPO,Dissolve]}
  n33 = ComponentMask(None=TextureCoordinate()) {r=true}
  n34 = ScalarParameter() {defaultValue=3, parameterName=EdgeIntensity}
  n35 = Saturate(None=Power(Base=Multiply(A=@n33, B=@n34) {constB=3}, Exp=ScalarParameter() {defaultValue=1, parameterName=TopMaskPower}) {constExponent=1})
  n36 = Multiply(A=Multiply(A=ScalarParameter() {defaultValue=20, parameterName=WidthBase}, B=@n32.SplineWidth), B=@n35)
  n37 = Multiply(A=Multiply(A=@n32.SplineWidth, B=ScalarParameter() {defaultValue=15, parameterName=WidthTip}) {constA=1}, B=@n35)
  n38 = StaticBool()
  n39 = VectorParameter() {parameterName=WPO_Direction, defaultValue=<rgba 8 0 0 0>}
  n40 = DynamicParameter() {defaultValue=<rgba 3.5 1 1 1>, paramNames [SplineWidth,U_offset,WPO,Dissolve]}
  n41 = ComponentMask(None=Add(A=TextureCoordinate(), B=DynamicParameter() {defaultValue=<rgba 3.5 1 1 1>, paramNames [SplineWidth,U_offset,WPO,Dissolve]}.U_offset)) {r=true}
  n42 = DynamicParameter() {parameterIndex=1, defaultValue=<rgba 3.5 0 1 1>, paramNames [PannerSpeed, Tex_Base_UOffset, None, None]}
  n43 = Multiply(A=@n42.PannerSpeed) {constB=-1}
  n44 = Multiply(
   A=Transform(None=Multiply(A=@n39.RGB, B=@n40.WPO)) {transformSourceType=TRANSFORMSOURCE_Local}
   B=Sine(None=Add(A=@n41, B=@n43)) {period=0.7}
  )
  n45 = Constant4Vector() {constant=<rgba 0 0 8 0>}
  n46 = DynamicParameter() {defaultValue=<rgba 3.5 1 1 1>, paramNames [SplineWidth,U_offset,WPO,Dissolve]}
  n47 = TextureCoordinate()
  n48 = DynamicParameter() {defaultValue=<rgba 3.5 1 1 1>, paramNames [SplineWidth,U_offset,WPO,Dissolve]}
  n49 = DynamicParameter() {parameterIndex=1, defaultValue=<rgba 3.5 0 1 1>, paramNames [PannerSpeed, Tex_Base_UOffset, None, None]}
  n50 = Add(A=ComponentMask(None=Add(A=@n47, B=@n48.U_offset)) {r=true}, B=Multiply(A=@n49.PannerSpeed) {constB=-0.5})
  n51 = Multiply(
   A=Transform(None=Multiply(A=@n45.RGB, B=@n46.WPO)) {transformSourceType=TRANSFORMSOURCE_Local}
   B=ConstantBiasScale(None=Sine(None=@n50) {period=0.6}) {scale=0.7}
  )
  n52 = ScalarParameter() {parameterName=Use SecondWPO}
  Reroute(None=Add(
    A=MaterialFunctionCall(WidthBase=@n36, WidthTip=@n37, Expand U or V UV Channel=@n38) {materialFunction=&SplineThicken}.WorldPositionOffset
    B=LinearInterpolate(A=@n44, B=@n51, Alpha=@n52)
   ))
  ScalarParameter() {defaultValue=3.5}
  Time()
  Time()
 ]
}
