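# Translucent, unlit, two-sided material for the start airflow effect of the LatteStir emote, used on Niagara mesh particles.
# Particle color tints the texture's green channel, which UseSimpleGlow can blend toward a flat value, before MF_FX_EmissiveAdjustments.
# Opacity multiplies a selectable texture channel by procedural UV fades at all four edges, a view-angle fade, depth fade and particle alpha.
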Material /BRCosmetics/Effects/Fort_Effects/Emotes/LatteStir/Materials/M_Emote_LatteStir_StartAirflow
graph {
 properties {
  bAutomaticallySetUsageInEditor=false
  bUsedWithNiagaraMeshParticles=true
  blendMode=BLEND_Translucent
  shadingModel=MSM_Unlit
  twoSided=true
  usedShadingModels=Unlit
 }
 nodes [
  n1 = TextureSampleParameter2D() {parameterName=BaseTexture, texture=&/T_Emote_LatteStir_StartAirflow}
  n2 = ScalarParameter() {parameterName=UseSimpleGlow}
  MaterialFunctionCall(Emissive=Multiply(A=ParticleColor().RGB, B=LinearInterpolate(A=@n1.G, Alpha=@n2))) {materialFunction=&MF_FX_EmissiveAdjustments}
  n3 = TextureSampleParameter2D() {parameterName=Tex_Alpha, texture=&T_Emote_LatteStir_StartAirflow}
  n4 = TextureCoordinate()
  n5 = Multiply(A=Power(Base=ComponentMask(None=@n4) {g=true})) {constB=100}
  n6 = ComponentMask(None=TextureCoordinate()) {r=true}
  n7 = Saturate(None=Multiply(A=OneMinus(None=@n6)) {constB=6})
  n8 = TextureCoordinate()
  n9 = Saturate(None=Multiply(A=ComponentMask(None=@n8) {r=true}) {constB=10})
  n10 = ComponentMask(None=TextureCoordinate()) {g=true}
  n11 = Multiply(A=Power(Base=OneMinus(None=@n10))) {constB=20}
  n12 = Multiply(
   A=Multiply(A=StaticComponentMaskParameter(None=@n3.RGB) {defaultR=true, parameterName=RGBMask}, B=Saturate(None=@n5))
   B=Multiply(A=Multiply(A=@n7, B=@n9), B=Saturate(None=@n11))
  )
  n13 = Constant() {r=1.5}
  n14 = Saturate(None=Multiply(A=MaterialFunctionCall("FadeSensitivity (optional)"=@n13) {materialFunction=&FadeBasedOnViewAngle}.Result) {constB=5})
  Multiply(A=ParticleColor().A, B=DepthFade(Opacity=Multiply(A=@n12, B=@n14)) {fadeDistanceDefault=30})
 ]
}
