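# Niagara effect for the start airflow of the LatteStir emote: red-tinted swirling mesh trails from a single DarkTrail emitter.
# An instant burst plus a low spawn rate creates particles drawn by two mesh renderers, with partition index picking one of four orientation offsets.
# Curves drive Z-axis spin and growing scale.
# Dynamic material parameters supply a random value and an early falloff over particle life.
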NiagaraSystem /BRCosmetics/Effects/Fort_Effects/Emotes/LatteStir/NS_Emote_LatteStir_StartAirflow
dependencies {
 usedDataInterfaces [&NiagaraDataInterfaceArrayFloat3, &NiagaraDataInterfaceCurve]
 usedDynamicInputs []
 usedModules [
  &self:SystemSpawnScript.SystemScriptSource.SystemScriptGraph.NiagaraNodeAssignment_0.SetVariables_#1
  &SystemState
 ]
}
emitters {DarkTrail001 {
  emitterUpdateScript [
   &EmitterState(
    Inactive Response=::NewEnumerator1
    Life Cycle Mode=::NewEnumerator1
    Loop Behavior=::NewEnumerator1
    Loop Delay=&Add_Float(A=$System.Delay, B=1.5)
    Loop Duration=3
    UseLoopDelay=false
   )
   &SpawnRate(SpawnRate=2, Use Spawn Probability=false)
   &SpawnBurst_Instantaneous(Spawn Count=3, Use Spawn Probability=false, UseLoopCountLimit=false)
  ]
  particleSpawnScript [
   &PartitionParticles(Number Of Partitions=4)
   &InitializeParticle(
    Color=<rgba 2 0 0.1667 1>
    Lifetime=2
    Mesh Scale Mode=::NewEnumerator1
    Mesh Uniform Scale Max=1.5
    Mesh Uniform Scale Min=0.8
    Position Offset=<0 0 20>
    UsePositionOffset=true
    Write Material Random=true
   )
   &InitialMeshOrientation(Mesh Orientation Mode=::NewEnumerator7, Use Rotation Vector=true) {module=InitialMeshOrientation001}
   &InitialMeshOrientation(
    Mesh Orientation Mode=::NewEnumerator2
    Rotation=&SelectVectorFromArray(
     Direct Array Index=$Particles.PartitionParticles.Partition
     Vector Selection Array=DataInterface() {propertyValues {FloatData [<0 0 0>, <0 0 0.3>, <0 0 0.6>, <0 0 0.8>]}}
    )
    Use Rotation Vector=true
   )
   &CameraOffset(Camera Offset Amount=10)
  ]
  particleUpdateScript [
   &ParticleState()
   &UpdateMeshOrientation(
    Rotation Rate=&FloatFromCurve(FloatCurve=DataInterface() {propertyValues {Curve {keys [<at 0 0>,<at 0.35 -2.3>]}}})
    Rotation Vector=<0 0 1>
   )
   &ScaleMeshSize(Scale Factor=&Multiply_VectorByFloat(
     Vector=&MakeVector(
      X=1
      Y=1
      Z=&FloatFromCurve(FloatCurve=DataInterface() {propertyValues {Curve {keys [
           {arriveTangent=-0.1793, interpMode=RCIM_Cubic, leaveTangent=-0.1793, tangentMode=RCTM_User, time=0, value=1}
           {arriveTangent=0.2568, tangentMode=RCTM_User, time=0.7, value=1.2}
          ]}}})
     )
     Float=&FloatFromCurve(
      FloatCurve=DataInterface() {propertyValues {Curve {
         defaultValue=1
         keys [
          {tangentMode=RCTM_User, time=0, value=0.5}
          {arriveTangent=1.75, interpMode=RCIM_Cubic, leaveTangent=1.75, time=0.15, value=1.0082}
          <at 0.4 1.2>
         ]
        }}}
      Scale Curve=&Multiply_Float(A=0.9, B=1)
     )
    ))
   &DynamicMaterialParameters(
    Index 0 Param 2=&RandomRangeFloat(Minimum=0, Maximum=1)
    Index 0 Param 4=&FloatFromCurve(FloatCurve=DataInterface() {propertyValues {Curve {keys [<at 0 1>,<at 0.4 0>,<at 1 0>]}}})
    Module.Index0Float0Write=true
    Module.Index0Float1Write=true
    Module.Index0Float2Write=true
    Module.Index0Float3Write=true
   )
   &FN_Mesh_Attributes(
    Color=&MakeLinearColorFromVectorAndFloat(
     "Vector (RGB)"=&MakeVectorFromLinearColorRGB(Color=<rgba 0.3 0 0 1>)
     "Float (Alpha)"=&Multiply_Float(A=1, B=2)
    )
    Create Color Attribute=true
    Write to Module Namespace=true
   ) {module=FN_Mesh_Attributes001}
   &SolveForcesAndVelocity(Clamp Velocity=false, Limit Acceleration=false)
  ]
  rendererClasses [&NiagaraMeshRendererProperties]
  renderers [
   {rendererClass=&NiagaraMeshRendererProperties, rendererIndex=0}
   {rendererClass=&NiagaraMeshRendererProperties, rendererIndex=1}
  ]
 }}
summary {
 systemName=NS_Emote_LatteStir_StartAirflow
 emitters [{emitterName=DarkTrail001, rendererClasses [&NiagaraMeshRendererProperties]}]
 userVariables [
  User.bIsFrontEnd {type=NiagaraBool}
  User.bIsFrontEndPreview {type=NiagaraBool}
  User.IsFrontEnd {type=NiagaraBool}
 ]
}
